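# Sound cue for Grandma's bodyfall-on-dirt foley, using creature step sound class with predator attenuation and foley concurrency.
# A quality-level switch feeds the first two levels from a random pool of six wave variations and the last two from a three-wave subset.
# A root modulator at default settings adds slight pitch and volume variation to the chosen wave.
# The three-wave random node is excluded from branch culling.
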SoundCue /Irwin/Sounds/Grandma/Foley/Foly_Grandma_Dirt_Bodyfall_Cue
root=Modulator_0
nodes [
 {id=Modulator_0, children [QualityLevel_0]}
 {id=QualityLevel_0, children [Random_1,Random_1,Random_0,Random_0]}
 {id=Random_0, children [WavePlayer_0,WavePlayer_1,WavePlayer_2], properties {bShouldExcludeFromBranchCulling=true, weights [1,1,1]}}
 {
  id=Random_1
  children [WavePlayer_6, WavePlayer_7, WavePlayer_8, WavePlayer_9, WavePlayer_10, WavePlayer_11]
  properties {weights [1,1,1,1,1,1]}
 }
 {id=WavePlayer_0, properties {soundWaveAssetPtr=&/Foly_Grandma_Dirt_Bodyfall_06}}
 {id=WavePlayer_1, properties {soundWaveAssetPtr=&/Foly_Grandma_Dirt_Bodyfall_01}}
 {id=WavePlayer_10, properties {soundWaveAssetPtr=&/Foly_Grandma_Dirt_Bodyfall_04}}
 {id=WavePlayer_11, properties {soundWaveAssetPtr=&/Foly_Grandma_Dirt_Bodyfall_05}}
 {id=WavePlayer_2, properties {soundWaveAssetPtr=&/Foly_Grandma_Dirt_Bodyfall_02}}
 {id=WavePlayer_6, properties {soundWaveAssetPtr=&Foly_Grandma_Dirt_Bodyfall_06}}
 {id=WavePlayer_7, properties {soundWaveAssetPtr=&Foly_Grandma_Dirt_Bodyfall_01}}
 {id=WavePlayer_8, properties {soundWaveAssetPtr=&Foly_Grandma_Dirt_Bodyfall_02}}
 {id=WavePlayer_9, properties {soundWaveAssetPtr=&/Foly_Grandma_Dirt_Bodyfall_03}}
]
properties {
 attenuationSettings=/Irwin/Sounds/Settings/Predator_Step_Default_Att
 maxDistance=4100
 soundClassObject=/Game/Sounds/FortSoundClassesAndModes/Fort_Creatures_Steps_Normal
 volumeMultiplier=1
 concurrencySet [/Irwin/Sounds/Settings/Conc/Predator_SetGroup_Foley_Conc]
}
