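# Sound cue for the rideable Grandma creature's foley when landing on dirt.
# A mixer layers random dirt-land and heavy-land waves under shared pitch and volume variation.
# IsLocal and IsRiddenByEnemy branches route that mix through different volume, sound class and attenuation, boosting the local case and giving enemy riders their own class and attenuation.
# Concurrency sets and overrides limit simultaneous plays.
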SoundCue /Irwin/Sounds/Grandma/Foley/Foly_Grandma_Dirt_Land_Cue
root=Branch_4
nodes [
 {id=Attenuation_0, children [Modulator_2], properties {attenuationSettings=/Irwin/Sounds/Settings/Predator_Gallop_Att}}
 {id=Attenuation_5, children [Modulator_2], properties {attenuationSettings=/Game/Sounds/Attenuation/Fort_Riding_Enemy_Att}}
 {id=Attenuation_6, children [Modulator_2], properties {attenuationSettings=&Predator_Gallop_Att}}
 {id=Attenuation_7, children [Modulator_2], properties {attenuationSettings=&Predator_Gallop_Att}}
 {id=Branch_3, children [Modulator_0,Modulator_4,SoundClass_0], properties {boolParameterName=IsRiddenByEnemy}}
 {id=Branch_4, children [Modulator_1,Branch_3,Branch_3], properties {boolParameterName=IsLocal}}
 {id=Mixer_0, children [Random_0,Random_2], properties {inputVolume [1.15,0.6]}}
 {id=Modulator_0, children [SoundClass_5], properties {pitchMax=1, pitchMin=1, volumeMax=0.7, volumeMin=0.7}}
 {id=Modulator_1, children [SoundClass_7], properties {pitchMax=1, pitchMin=1, volumeMax=1.5, volumeMin=1.5}}
 {id=Modulator_2, children [Mixer_0], properties {pitchMax=1.2, pitchMin=0.9, volumeMax=1, volumeMin=0.9}}
 {id=Modulator_4, children [SoundClass_6], properties {pitchMax=1, pitchMin=1, volumeMax=0.3, volumeMin=0.3}}
 {
  id=Random_0
  children [WavePlayer_0,WavePlayer_1,WavePlayer_2,WavePlayer_3,WavePlayer_4,WavePlayer_5]
  properties {bShouldExcludeFromBranchCulling=true, weights [1,1,1,1,1,1]}
 }
 {
  id=Random_2
  children [WavePlayer_9, WavePlayer_10, WavePlayer_11, WavePlayer_12, WavePlayer_13, WavePlayer_14]
  properties {bShouldExcludeFromBranchCulling=true, weights [1,1,1,1,1,1]}
 }
 {
  id=SoundClass_0
  children [Attenuation_0]
  properties {soundClassOverride=/Game/Sounds/FortSoundClassesAndModes/Fort_Creatures_Steps_Tactical}
 }
 {
  id=SoundClass_5
  children [Attenuation_5]
  properties {soundClassOverride=&Fort_Player_Foley_3P_Enemy_Athena}
 }
 {
  id=SoundClass_6
  children [Attenuation_6]
  properties {soundClassOverride=&Fort_Creatures_Steps_Tactical}
 }
 {
  id=SoundClass_7
  children [Attenuation_7]
  properties {soundClassOverride=&Fort_Creatures_Steps_Tactical}
 }
 {id=WavePlayer_0, properties {soundWaveAssetPtr=&/Foly_Grandma_Dirt_Land_01}}
 {id=WavePlayer_1, properties {soundWaveAssetPtr=&/Foly_Grandma_Dirt_Land_02}}
 {id=WavePlayer_10, properties {soundWaveAssetPtr=&/Grandma_Land_Heavy_01}}
 {id=WavePlayer_11, properties {soundWaveAssetPtr=&/Grandma_Land_Heavy_02}}
 {id=WavePlayer_12, properties {soundWaveAssetPtr=&/Grandma_Land_Heavy_03}}
 {id=WavePlayer_13, properties {soundWaveAssetPtr=&/Grandma_Land_Heavy_04}}
 {id=WavePlayer_14, properties {soundWaveAssetPtr=&/Grandma_Land_Heavy_05}}
 {id=WavePlayer_2, properties {soundWaveAssetPtr=&/Foly_Grandma_Dirt_Land_03}}
 {id=WavePlayer_3, properties {soundWaveAssetPtr=&/Foly_Grandma_Dirt_Land_04}}
 {id=WavePlayer_4, properties {soundWaveAssetPtr=&/Foly_Grandma_Dirt_Land_05}}
 {id=WavePlayer_5, properties {soundWaveAssetPtr=&/Foly_Grandma_Dirt_Land_06}}
 {id=WavePlayer_9, properties {soundWaveAssetPtr=&/Grandma_Land_Heavy_06}}
]
properties {
 bOverrideConcurrency=true
 maxDistance=13000
 volumeMultiplier=1.5
 concurrencyOverrides {maxCount=3}
 concurrencySet [/Irwin/Sounds/Settings/Conc/Predator_SetGroup_Foley_Conc, /Irwin/Sounds/Settings/Conc/SetOneLimitToOwner_Foley_Conc]
}
