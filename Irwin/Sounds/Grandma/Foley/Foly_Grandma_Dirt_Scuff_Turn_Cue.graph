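# Sound cue for Grandma creature foley, playing randomized dirt scuff sounds for turning, with a Predator foley concurrency group.
# IsLocal and IsRiddenByEnemy branches select sound class, attenuation and volume scaling, with a separate enemy-rider path for non-local playback.
# All paths converge on one volume modulator and a quality-level switch between a seven-wave random pool and a four-wave subset.
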SoundCue /Irwin/Sounds/Grandma/Foley/Foly_Grandma_Dirt_Scuff_Turn_Cue
root=Branch_4
nodes [
 {id=Attenuation_0, children [Modulator_0], properties {attenuationSettings=/Game/Sounds/Attenuation/Fort_Riding_Enemy_Att}}
 {id=Attenuation_1, children [Modulator_0], properties {attenuationSettings=/Irwin/Sounds/Settings/Predator_Step_Default_Att}}
 {id=Attenuation_2, children [Modulator_0], properties {attenuationSettings=&Predator_Step_Default_Att}}
 {id=Attenuation_7, children [Modulator_0], properties {attenuationSettings=&Predator_Step_Default_Att}}
 {id=Branch_0, children [Modulator_1,Modulator_4,SoundClass_2], properties {boolParameterName=IsRiddenByEnemy}}
 {id=Branch_4, children [SoundClass_7,Branch_0,Branch_0], properties {boolParameterName=IsLocal}}
 {id=Modulator_0, children [QualityLevel_0], properties {volumeMax=1, volumeMin=0.6}}
 {id=Modulator_1, children [SoundClass_0], properties {pitchMax=1, pitchMin=1, volumeMax=0.5, volumeMin=0.5}}
 {id=Modulator_4, children [SoundClass_1], properties {pitchMax=1, pitchMin=1, volumeMax=0.3, volumeMin=0.3}}
 {id=QualityLevel_0, children [Random_0,Random_0,Random_1,Random_1]}
 {
  id=Random_0
  children [WavePlayer_0,WavePlayer_1,WavePlayer_2,WavePlayer_3,WavePlayer_4,WavePlayer_5,WavePlayer_6]
  properties {weights [1,1,1,1,1,1,1]}
 }
 {
  id=Random_1
  children [WavePlayer_7, WavePlayer_8, WavePlayer_9, WavePlayer_10]
  properties {bShouldExcludeFromBranchCulling=true, weights [1,1,1,1]}
 }
 {
  id=SoundClass_0
  children [Attenuation_0]
  properties {soundClassOverride=&Fort_Player_Foley_3P_Enemy_Athena}
 }
 {
  id=SoundClass_1
  children [Attenuation_1]
  properties {soundClassOverride=/Game/Sounds/FortSoundClassesAndModes/Fort_Creatures_Steps_Tactical}
 }
 {
  id=SoundClass_2
  children [Attenuation_2]
  properties {soundClassOverride=&Fort_Creatures_Steps_Tactical}
 }
 {
  id=SoundClass_7
  children [Attenuation_7]
  properties {soundClassOverride=&Fort_Creatures_Steps_Tactical}
 }
 {id=WavePlayer_0, properties {soundWaveAssetPtr=&/Foly_Grandma_Dirt_Scuff_Turn_08}}
 {id=WavePlayer_1, properties {soundWaveAssetPtr=&/Foly_Grandma_Dirt_Scuff_Turn_01}}
 {id=WavePlayer_10, properties {soundWaveAssetPtr=&/Foly_Grandma_Dirt_Scuff_Turn_03}}
 {id=WavePlayer_2, properties {soundWaveAssetPtr=&/Foly_Grandma_Dirt_Scuff_Turn_02}}
 {id=WavePlayer_3, properties {soundWaveAssetPtr=&Foly_Grandma_Dirt_Scuff_Turn_03}}
 {id=WavePlayer_4, properties {soundWaveAssetPtr=&/Foly_Grandma_Dirt_Scuff_Turn_04}}
 {id=WavePlayer_5, properties {soundWaveAssetPtr=&/Foly_Grandma_Dirt_Scuff_Turn_05}}
 {id=WavePlayer_6, properties {soundWaveAssetPtr=&/Foly_Grandma_Dirt_Scuff_Turn_06}}
 {id=WavePlayer_7, properties {soundWaveAssetPtr=&Foly_Grandma_Dirt_Scuff_Turn_08}}
 {id=WavePlayer_8, properties {soundWaveAssetPtr=&Foly_Grandma_Dirt_Scuff_Turn_01}}
 {id=WavePlayer_9, properties {soundWaveAssetPtr=&Foly_Grandma_Dirt_Scuff_Turn_02}}
]
properties {maxDistance=13000, concurrencySet [/Irwin/Sounds/Settings/Conc/Predator_SetGroup_Foley_Conc, null]}
