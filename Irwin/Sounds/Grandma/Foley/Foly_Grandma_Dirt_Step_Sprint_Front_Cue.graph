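# Sound cue for Grandma creature's front-foot sprint steps on dirt.
# IsLocal plays at full branch level; otherwise IsRiddenByEnemy selects a louder enemy foley class and riding attenuation when true, or quieter creature steps when false.
# All paths share random volume modulation and a quality switch between six- and four-variation random pools.
# Two concurrency sets limit overlapping playback.
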SoundCue /Irwin/Sounds/Grandma/Foley/Foly_Grandma_Dirt_Step_Sprint_Front_Cue
root=Branch_4
nodes [
 {id=Attenuation_0, children [Modulator_2], properties {attenuationSettings=/Irwin/Sounds/Settings/Predator_Gallop_Att}}
 {id=Attenuation_5, children [Modulator_2], properties {attenuationSettings=/Game/Sounds/Attenuation/Fort_Riding_Enemy_Att}}
 {id=Attenuation_6, children [Modulator_2], properties {attenuationSettings=&Predator_Gallop_Att}}
 {id=Attenuation_7, children [Modulator_2], properties {attenuationSettings=&Predator_Gallop_Att}}
 {id=Branch_3, children [Modulator_1,Modulator_4,SoundClass_0], properties {boolParameterName=IsRiddenByEnemy}}
 {id=Branch_4, children [SoundClass_7,Branch_3,Branch_3], properties {boolParameterName=IsLocal}}
 {id=Modulator_1, children [SoundClass_5], properties {pitchMax=1, pitchMin=1, volumeMax=0.7, volumeMin=0.7}}
 {id=Modulator_2, children [QualityLevel_0], properties {volumeMax=1, volumeMin=0.5}}
 {id=Modulator_4, children [SoundClass_6], properties {pitchMax=1, pitchMin=1, volumeMax=0.3, volumeMin=0.3}}
 {id=QualityLevel_0, children [Random_1,Random_1,Random_0,Random_0]}
 {
  id=Random_0
  children [WavePlayer_0,WavePlayer_1,WavePlayer_2,WavePlayer_3]
  properties {bShouldExcludeFromBranchCulling=true, weights [1,1,1,1]}
 }
 {
  id=Random_1
  children [WavePlayer_13, WavePlayer_14, WavePlayer_15, WavePlayer_16, WavePlayer_17, WavePlayer_18]
  properties {weights [1,1,1,1,1,1]}
 }
 {
  id=SoundClass_0
  children [Attenuation_0]
  properties {soundClassOverride=/Game/Sounds/FortSoundClassesAndModes/Fort_Creatures_Steps_Tactical}
 }
 {
  id=SoundClass_5
  children [Attenuation_5]
  properties {soundClassOverride=&Fort_Player_Foley_3P_Enemy_Athena}
 }
 {
  id=SoundClass_6
  children [Attenuation_6]
  properties {soundClassOverride=&Fort_Creatures_Steps_Tactical}
 }
 {
  id=SoundClass_7
  children [Attenuation_7]
  properties {soundClassOverride=&Fort_Creatures_Steps_Tactical}
 }
 {id=WavePlayer_0, properties {soundWaveAssetPtr=&/Foly_Grandma_Dirt_Step_Sprint_Front_01}}
 {id=WavePlayer_1, properties {soundWaveAssetPtr=&/Foly_Grandma_Dirt_Step_Sprint_Front_02}}
 {id=WavePlayer_13, properties {soundWaveAssetPtr=&Foly_Grandma_Dirt_Step_Sprint_Front_01}}
 {id=WavePlayer_14, properties {soundWaveAssetPtr=&Foly_Grandma_Dirt_Step_Sprint_Front_02}}
 {id=WavePlayer_15, properties {soundWaveAssetPtr=&/Foly_Grandma_Dirt_Step_Sprint_Front_03}}
 {id=WavePlayer_16, properties {soundWaveAssetPtr=&/Foly_Grandma_Dirt_Step_Sprint_Front_04}}
 {id=WavePlayer_17, properties {soundWaveAssetPtr=&/Foly_Grandma_Dirt_Step_Sprint_Front_06}}
 {id=WavePlayer_18, properties {soundWaveAssetPtr=&/Foly_Grandma_Dirt_Step_Sprint_Front_07}}
 {id=WavePlayer_2, properties {soundWaveAssetPtr=&Foly_Grandma_Dirt_Step_Sprint_Front_03}}
 {id=WavePlayer_3, properties {soundWaveAssetPtr=&Foly_Grandma_Dirt_Step_Sprint_Front_04}}
]
properties {maxDistance=13000, concurrencySet [/Irwin/Sounds/Settings/Conc/Predator_SetGroup_Foley_Conc, &/Grandma_Run_Conc]}
