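# Sound cue for Grandma's shake movement foley in the Irwin content, routed to the creature foley sound class.
# A Random node picks one of three equally weighted shake wave variants and feeds a Modulator that pins pitch at its original value.
# Playback uses the shared Predator step attenuation and Predator foley concurrency settings.
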SoundCue /Irwin/Sounds/Grandma/Foley/Foly_Grandma_Mvmnt_Shake_Cue
root=Modulator_0
nodes [
 {id=Modulator_0, children [Random_0], properties {pitchMax=1, pitchMin=1}}
 {id=Random_0, children [WavePlayer_0,WavePlayer_1,WavePlayer_2], properties {bShouldExcludeFromBranchCulling=true, weights [1,1,1]}}
 {id=WavePlayer_0, properties {soundWaveAssetPtr=&/Foly_Grandma_Mvmnt_Shake_01}}
 {id=WavePlayer_1, properties {soundWaveAssetPtr=&/Foly_Grandma_Mvmnt_Shake_02}}
 {id=WavePlayer_2, properties {soundWaveAssetPtr=&/Foly_Grandma_Mvmnt_Shake_03}}
]
properties {
 attenuationSettings=/Irwin/Sounds/Settings/Predator_Step_Default_Att
 maxDistance=4100
 soundClassObject=/Game/Sounds/FortSoundClassesAndModes/Fort_Creatures_Foley
 concurrencySet [/Irwin/Sounds/Settings/Conc/Predator_SetGroup_Foley_Conc]
}
