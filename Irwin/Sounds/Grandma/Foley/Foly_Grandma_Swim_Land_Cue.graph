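# Sound cue for the Grandma creature's foley when landing from a swim, under a predator foley concurrency group.
# IsLocal plays a boosted path; otherwise IsRiddenByEnemy picks an enemy-rider sound class and attenuation, a much quieter path, or unmodified volume when unset.
# All paths share a pitch and volume randomizer into a quality switch between four-wave and two-wave random pools, the smaller excluded from branch culling.
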SoundCue /Irwin/Sounds/Grandma/Foley/Foly_Grandma_Swim_Land_Cue
root=Branch_4
nodes [
 {id=Attenuation_0, children [Modulator_2], properties {attenuationSettings=/Irwin/Sounds/Settings/Predator_Step_Default_Att}}
 {id=Attenuation_5, children [Modulator_2], properties {attenuationSettings=/Game/Sounds/Attenuation/Fort_Riding_Enemy_Att}}
 {id=Attenuation_6, children [Modulator_2], properties {attenuationSettings=&Predator_Step_Default_Att}}
 {id=Attenuation_7, children [Modulator_2], properties {attenuationSettings=&Predator_Step_Default_Att}}
 {id=Branch_3, children [Modulator_0,Modulator_4,SoundClass_0], properties {boolParameterName=IsRiddenByEnemy}}
 {id=Branch_4, children [Modulator_1,Branch_3,Branch_3], properties {boolParameterName=IsLocal}}
 {id=Modulator_0, children [SoundClass_5], properties {pitchMax=1, pitchMin=1, volumeMax=0.6, volumeMin=0.6}}
 {id=Modulator_1, children [SoundClass_7], properties {pitchMax=1, pitchMin=1, volumeMax=1.25, volumeMin=1.25}}
 {id=Modulator_2, children [QualityLevel_0], properties {pitchMax=1.2, volumeMax=1, volumeMin=0.8}}
 {id=Modulator_4, children [SoundClass_6], properties {pitchMax=1, pitchMin=1, volumeMax=0.1, volumeMin=0.1}}
 {id=QualityLevel_0, children [Random_0,Random_0,Random_1,Random_1]}
 {id=Random_0, children [WavePlayer_2,WavePlayer_3,WavePlayer_4,WavePlayer_5], properties {weights [1,1,1,1]}}
 {id=Random_1, children [WavePlayer_0,WavePlayer_1], properties {bShouldExcludeFromBranchCulling=true, weights [1,1]}}
 {
  id=SoundClass_0
  children [Attenuation_0]
  properties {soundClassOverride=/Game/Sounds/FortSoundClassesAndModes/Fort_Creatures_Steps_Tactical}
 }
 {
  id=SoundClass_5
  children [Attenuation_5]
  properties {soundClassOverride=&Fort_Player_Foley_3P_Enemy_Athena}
 }
 {
  id=SoundClass_6
  children [Attenuation_6]
  properties {soundClassOverride=&Fort_Creatures_Steps_Tactical}
 }
 {
  id=SoundClass_7
  children [Attenuation_7]
  properties {soundClassOverride=&Fort_Creatures_Steps_Tactical}
 }
 {id=WavePlayer_0, properties {soundWaveAssetPtr=&/Foly_Grandma_Swim_Land_01}}
 {id=WavePlayer_1, properties {soundWaveAssetPtr=&/Foly_Grandma_Swim_Land_02}}
 {id=WavePlayer_2, properties {soundWaveAssetPtr=&Foly_Grandma_Swim_Land_01}}
 {id=WavePlayer_3, properties {soundWaveAssetPtr=&Foly_Grandma_Swim_Land_02}}
 {id=WavePlayer_4, properties {soundWaveAssetPtr=&/Foly_Grandma_Swim_Land_03}}
 {id=WavePlayer_5, properties {soundWaveAssetPtr=&/Foly_Grandma_Swim_Land_04}}
]
properties {maxDistance=13000, volumeMultiplier=3, concurrencySet [/Irwin/Sounds/Settings/Conc/Predator_SetGroup_Foley_Conc]}
