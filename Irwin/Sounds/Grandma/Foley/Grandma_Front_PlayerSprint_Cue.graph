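# Sound cue for the front sprint footsteps of the Grandma creature, layering dirt steps with quieter heavy sprint sounds.
# IsLocal and IsRiddenByEnemy branches set volume, sound class and attenuation, making local loudest and enemy-ridden louder than other remote cases.
# All branches share a volume randomizer and a quality-level switch between six- and four-variant random mixes.
# Concurrency sets limit playback, and the whole cue is pitched up.
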SoundCue /Irwin/Sounds/Grandma/Foley/Grandma_Front_PlayerSprint_Cue
root=Branch_4
nodes [
 {id=Attenuation_0, children [Modulator_0], properties {attenuationSettings=/Game/Sounds/Attenuation/Fort_Riding_Enemy_Att}}
 {id=Attenuation_1, children [Modulator_0], properties {attenuationSettings=/Irwin/Sounds/Settings/Predator_Gallop_Att}}
 {id=Attenuation_7, children [Modulator_0], properties {attenuationSettings=&Predator_Gallop_Att}}
 {id=Branch_0, children [Modulator_1,Modulator_4,Modulator_4], properties {boolParameterName=IsRiddenByEnemy}}
 {id=Branch_4, children [Modulator_2,Branch_0,Branch_0], properties {boolParameterName=IsLocal}}
 {id=Mixer_2, children [Random_1,Random_5], properties {inputVolume [1,0.7]}}
 {id=Mixer_3, children [Random_0,Random_4], properties {inputVolume [1,0.7]}}
 {id=Modulator_0, children [QualityLevel_0], properties {volumeMax=1, volumeMin=0.6}}
 {id=Modulator_1, children [SoundClass_0], properties {pitchMax=1, pitchMin=1, volumeMax=0.7, volumeMin=0.7}}
 {id=Modulator_2, children [SoundClass_7], properties {pitchMax=1.1, pitchMin=1, volumeMax=1.2, volumeMin=0.9}}
 {id=Modulator_4, children [SoundClass_1], properties {pitchMax=1, pitchMin=1, volumeMax=0.3, volumeMin=0.3}}
 {id=QualityLevel_0, children [Mixer_2,Mixer_2,Mixer_3,Mixer_3]}
 {
  id=Random_0
  children [WavePlayer_30, WavePlayer_31, WavePlayer_32, WavePlayer_33]
  properties {bShouldExcludeFromBranchCulling=true, weights [1,1,1,1]}
 }
 {
  id=Random_1
  children [WavePlayer_24, WavePlayer_25, WavePlayer_26, WavePlayer_27, WavePlayer_28, WavePlayer_29]
  properties {weights [1,1,1,1,1,1]}
 }
 {
  id=Random_4
  children [WavePlayer_34, WavePlayer_35, WavePlayer_36, WavePlayer_37]
  properties {bShouldExcludeFromBranchCulling=true, weights [1,1,1,1]}
 }
 {
  id=Random_5
  children [WavePlayer_4, WavePlayer_5, WavePlayer_6, WavePlayer_7, WavePlayer_8, WavePlayer_23]
  properties {weights [1,1,1,1,1,1]}
 }
 {
  id=SoundClass_0
  children [Attenuation_0]
  properties {soundClassOverride=&Fort_Player_Foley_3P_Enemy_Athena}
 }
 {
  id=SoundClass_1
  children [Attenuation_1]
  properties {soundClassOverride=/Game/Sounds/FortSoundClassesAndModes/Fort_Creatures_Steps_Tactical}
 }
 {
  id=SoundClass_7
  children [Attenuation_7]
  properties {soundClassOverride=&Fort_Creatures_Steps_Tactical}
 }
 {id=WavePlayer_23, properties {soundWaveAssetPtr=&/Grandma_Sprint_Front_Heavy_06}}
 {id=WavePlayer_24, properties {soundWaveAssetPtr=&/Foly_Grandma_Dirt_Step_Sprint_Front_01}}
 {id=WavePlayer_25, properties {soundWaveAssetPtr=&/Foly_Grandma_Dirt_Step_Sprint_Front_02}}
 {id=WavePlayer_26, properties {soundWaveAssetPtr=&/Foly_Grandma_Dirt_Step_Sprint_Front_03}}
 {id=WavePlayer_27, properties {soundWaveAssetPtr=&/Foly_Grandma_Dirt_Step_Sprint_Front_04}}
 {id=WavePlayer_28, properties {soundWaveAssetPtr=&/Foly_Grandma_Dirt_Step_Sprint_Front_06}}
 {id=WavePlayer_29, properties {soundWaveAssetPtr=&/Foly_Grandma_Dirt_Step_Sprint_Front_07}}
 {id=WavePlayer_30, properties {soundWaveAssetPtr=&Foly_Grandma_Dirt_Step_Sprint_Front_01}}
 {id=WavePlayer_31, properties {soundWaveAssetPtr=&Foly_Grandma_Dirt_Step_Sprint_Front_02}}
 {id=WavePlayer_32, properties {soundWaveAssetPtr=&Foly_Grandma_Dirt_Step_Sprint_Front_03}}
 {id=WavePlayer_33, properties {soundWaveAssetPtr=&Foly_Grandma_Dirt_Step_Sprint_Front_04}}
 {id=WavePlayer_34, properties {soundWaveAssetPtr=&/Grandma_Sprint_Front_Heavy_01}}
 {id=WavePlayer_35, properties {soundWaveAssetPtr=&/Grandma_Sprint_Front_Heavy_02}}
 {id=WavePlayer_36, properties {soundWaveAssetPtr=&/Grandma_Sprint_Front_Heavy_03}}
 {id=WavePlayer_37, properties {soundWaveAssetPtr=&/Grandma_Sprint_Front_Heavy_04}}
 {id=WavePlayer_4, properties {soundWaveAssetPtr=&Grandma_Sprint_Front_Heavy_01}}
 {id=WavePlayer_5, properties {soundWaveAssetPtr=&Grandma_Sprint_Front_Heavy_02}}
 {id=WavePlayer_6, properties {soundWaveAssetPtr=&Grandma_Sprint_Front_Heavy_03}}
 {id=WavePlayer_7, properties {soundWaveAssetPtr=&Grandma_Sprint_Front_Heavy_04}}
 {id=WavePlayer_8, properties {soundWaveAssetPtr=&/Grandma_Sprint_Front_Heavy_05}}
]
properties {
 maxDistance=13000
 pitchMultiplier=1.3
 volumeMultiplier=0.9
 concurrencySet [/Irwin/Sounds/Settings/Conc/SetOneLimitToOwner_Foley_Conc, /Irwin/Sounds/Settings/Conc/Prey_SetGroup_Foley_Conc]
}
