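# Sound cue for the rear footstep foley of the Grandma creature's player sprint, pitched up and limited by two concurrency sets.
# IsLocal and IsRiddenByEnemy branches select sound class, attenuation and volume, making enemy-ridden steps louder than other remote ones.
# All paths share a quality switch between seven- and four-variation mixes of random dirt steps and heavy layers.
# Each mix also has a looping wind layer gated on IsLLocal, spelled differently from IsLocal.
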SoundCue /Irwin/Sounds/Grandma/Foley/Grandma_Rear_PlayerSprint_Cue
root=Branch_4
nodes [
 {id=Attenuation_0, children [Modulator_0], properties {attenuationSettings=/Game/Sounds/Attenuation/Fort_Riding_Enemy_Att}}
 {id=Attenuation_1, children [Modulator_0], properties {attenuationSettings=/Irwin/Sounds/Settings/Predator_Gallop_Att}}
 {id=Attenuation_7, children [Modulator_0], properties {attenuationSettings=&Predator_Gallop_Att}}
 {id=Branch_0, children [Modulator_1,Modulator_4,Modulator_4], properties {boolParameterName=IsRiddenByEnemy}}
 {id=Branch_1, children [WavePlayer_4,null,null], properties {boolParameterName=IsLLocal}}
 {id=Branch_2, children [WavePlayer_5,null,null], properties {boolParameterName=IsLLocal}}
 {id=Branch_4, children [SoundClass_7,Branch_0,Branch_0], properties {boolParameterName=IsLocal}}
 {id=Mixer_2, children [Random_2,Random_3,Branch_1], properties {inputVolume [1,1.2,1]}}
 {id=Mixer_3, children [Random_6,Random_7,Branch_2], properties {inputVolume [1,1.2,1]}}
 {id=Modulator_0, children [QualityLevel_0], properties {volumeMax=1, volumeMin=0.5}}
 {id=Modulator_1, children [SoundClass_0], properties {pitchMax=1, pitchMin=1, volumeMax=0.7, volumeMin=0.7}}
 {id=Modulator_4, children [SoundClass_1], properties {pitchMax=1, pitchMin=1, volumeMax=0.3, volumeMin=0.3}}
 {id=QualityLevel_0, children [Mixer_2,Mixer_2,Mixer_3,Mixer_3]}
 {
  id=Random_2
  children [WavePlayer_12, WavePlayer_13, WavePlayer_14, WavePlayer_15, WavePlayer_16, WavePlayer_17, WavePlayer_18]
  properties {weights [1,1,1,1,1,1,1]}
 }
 {
  id=Random_3
  children [WavePlayer_0, WavePlayer_1, WavePlayer_2, WavePlayer_3, WavePlayer_9, WavePlayer_10, WavePlayer_11]
  properties {weights [1,1,1,1,1,1,1]}
 }
 {
  id=Random_6
  children [WavePlayer_41, WavePlayer_42, WavePlayer_43, WavePlayer_44]
  properties {bShouldExcludeFromBranchCulling=true, weights [1,1,1,1]}
 }
 {
  id=Random_7
  children [WavePlayer_19, WavePlayer_20, WavePlayer_21, WavePlayer_22]
  properties {bShouldExcludeFromBranchCulling=true, weights [1,1,1,1]}
 }
 {
  id=SoundClass_0
  children [Attenuation_0]
  properties {soundClassOverride=&Fort_Player_Foley_3P_Enemy_Athena}
 }
 {
  id=SoundClass_1
  children [Attenuation_1]
  properties {soundClassOverride=/Game/Sounds/FortSoundClassesAndModes/Fort_Creatures_Steps_Tactical}
 }
 {
  id=SoundClass_7
  children [Attenuation_7]
  properties {soundClassOverride=&Fort_Creatures_Steps_Tactical}
 }
 {id=WavePlayer_0, properties {soundWaveAssetPtr=&/Grandma_Sprint_Rear_Heavy_01}}
 {id=WavePlayer_1, properties {soundWaveAssetPtr=&/Grandma_Sprint_Rear_Heavy_02}}
 {id=WavePlayer_10, properties {soundWaveAssetPtr=&/Grandma_Sprint_Rear_Heavy_06}}
 {id=WavePlayer_11, properties {soundWaveAssetPtr=&/Grandma_Sprint_Rear_Heavy_07}}
 {id=WavePlayer_12, properties {soundWaveAssetPtr=&/Foly_Grandma_Dirt_Step_Sprint_Rear_07}}
 {id=WavePlayer_13, properties {soundWaveAssetPtr=&/Foly_Grandma_Dirt_Step_Sprint_Rear_01}}
 {id=WavePlayer_14, properties {soundWaveAssetPtr=&/Foly_Grandma_Dirt_Step_Sprint_Rear_02}}
 {id=WavePlayer_15, properties {soundWaveAssetPtr=&/Foly_Grandma_Dirt_Step_Sprint_Rear_03}}
 {id=WavePlayer_16, properties {soundWaveAssetPtr=&/Foly_Grandma_Dirt_Step_Sprint_Rear_04}}
 {id=WavePlayer_17, properties {soundWaveAssetPtr=&/Foly_Grandma_Dirt_Step_Sprint_Rear_05}}
 {id=WavePlayer_18, properties {soundWaveAssetPtr=&/Foly_Grandma_Dirt_Step_Sprint_Rear_06}}
 {id=WavePlayer_19, properties {soundWaveAssetPtr=&Grandma_Sprint_Rear_Heavy_01}}
 {id=WavePlayer_2, properties {soundWaveAssetPtr=&/Grandma_Sprint_Rear_Heavy_03}}
 {id=WavePlayer_20, properties {soundWaveAssetPtr=&Grandma_Sprint_Rear_Heavy_02}}
 {id=WavePlayer_21, properties {soundWaveAssetPtr=&Grandma_Sprint_Rear_Heavy_03}}
 {id=WavePlayer_22, properties {soundWaveAssetPtr=&/Grandma_Sprint_Rear_Heavy_04}}
 {id=WavePlayer_3, properties {soundWaveAssetPtr=&Grandma_Sprint_Rear_Heavy_04}}
 {id=WavePlayer_4, properties {bLooping=true, soundWaveAssetPtr=/Game/Athena/Sounds/Wind_Falling/AMB_Wind_Skydive_Vertical_Loop_01}}
 {id=WavePlayer_41, properties {soundWaveAssetPtr=&Foly_Grandma_Dirt_Step_Sprint_Rear_07}}
 {id=WavePlayer_42, properties {soundWaveAssetPtr=&Foly_Grandma_Dirt_Step_Sprint_Rear_01}}
 {id=WavePlayer_43, properties {soundWaveAssetPtr=&Foly_Grandma_Dirt_Step_Sprint_Rear_02}}
 {id=WavePlayer_44, properties {soundWaveAssetPtr=&Foly_Grandma_Dirt_Step_Sprint_Rear_03}}
 {id=WavePlayer_5, properties {bLooping=true, soundWaveAssetPtr=&AMB_Wind_Skydive_Vertical_Loop_01}}
 {id=WavePlayer_9, properties {soundWaveAssetPtr=&/Grandma_Sprint_Rear_Heavy_05}}
]
properties {
 maxDistance=13000
 pitchMultiplier=1.3
 volumeMultiplier=0.9
 concurrencySet [/Irwin/Sounds/Settings/Conc/SetOneLimitToOwner_Foley_Conc, /Irwin/Sounds/Settings/Conc/Prey_SetGroup_Foley_Conc]
}
